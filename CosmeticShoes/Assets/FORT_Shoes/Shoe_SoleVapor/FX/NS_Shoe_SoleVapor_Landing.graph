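# Niagara system for the SoleVapor cosmetic shoe's landing effect, bursting smoke, sparkles and rainbow trails.
# Mesh emitters make ground smoke and colored per-foot puffs, while sprites add flickering sparkles.
# Two ribbons follow skeletal mesh foot sockets, stretched by lagged socket motion and scaled by owner speed.
# User parameters suppress the effect, cull feet with small ShoeScale values, and adjust preview items, which omit ground smoke.
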NiagaraSystem /CosmeticShoes/Assets/FORT_Shoes/Shoe_SoleVapor/FX/NS_Shoe_SoleVapor_Landing
dependencies {
 usedDataInterfaces [
  &NiagaraDataInterfaceCurve
  &NiagaraDataInterfaceArrayFloat
  &NiagaraDataInterfaceSkeletalMesh
  &NiagaraDataInterfaceColorCurve
 ]
 usedDynamicInputs [&ColorFromCurve, &InvertBool, &RemapRange]
 usedModules [
  &self:SystemSpawnScript.SystemScriptSource.SystemScriptGraph.NiagaraNodeAssignment_0.SetVariables_#1
  &self:SystemSpawnScript.SystemScriptSource.SystemScriptGraph.NiagaraNodeAssignment_1.SetVariables_#2
  &self:SystemSpawnScript.SystemScriptSource.SystemScriptGraph.NiagaraNodeAssignment_2.SetVariables_#3
  &SystemState
  &self:SystemSpawnScript.SystemScriptSource.SystemScriptGraph.NiagaraNodeAssignment_3.SetVariables_#4
  &self:SystemSpawnScript.SystemScriptSource.SystemScriptGraph.NiagaraNodeAssignment_4.SetVariables_#5
 ]
}
emitters {
 ColoredSmokePuffs {
  emitterUpdateScript [
   &EmitterState(Life Cycle Mode=::NewEnumerator1, Loop Behavior=::NewEnumerator1, Loop Duration Mode=::NewEnumerator1, UseLoopDelay=false)
   &SpawnBurst_Instantaneous(
    Spawn Count=&RandomRangeInt(Minimum=12, Maximum=18)
    Spawn Probability=&MakeCustomFloatFromBool(Bool=$User.HideFootFX, True Float=0, False Float=1)
    Use Spawn Probability=true
    UseLoopCountLimit=false
   )
  ]
  particleSpawnScript [
   &PartitionParticles(Write Partition to Ribbon ID=::NewEnumerator0)
   &InitializeParticle(
    Color=&MergeHelper_ColorSwitch(
     0=<rgba 0 0.4733 1 1>
     1=<rgba 0 0.8281 0.3672 1>
     2=<rgba 1 0.3109 0 1>
     3=<rgba 1 0.0677 0.099 1>
     Index=&RandomRangeInt(Minimum=0, Maximum=3)
     Num=::NewEnumerator2
    )
    Lifetime=&MakeCustomFloatFromBool(
     Bool=$User.bisPrimaryPreviewItem
     True Float=&RandomRangeFloat(Minimum=0.4, Maximum=0.6)
     False Float=&RandomRangeFloat(Minimum=0.3, Maximum=0.6)
    )
    Mesh Scale Mode=::NewEnumerator1
    Mesh Uniform Scale Max=0.3
    Mesh Uniform Scale Min=0.1
    UsePositionOffset=false
    Write Material Random=true
   )
   &SkeletalMeshLocation(
    Bone / Socket Index=$Particles.PartitionParticles.Partition
    Bone Sampling Mode=::NewEnumerator4
    MaskSpawnGroup=false
    Orientation Sampling=::NewEnumerator1
    Position Sampling=::NewEnumerator2
    Sampled Position Offset=&Multiply_VectorByFloat(
     Vector=&Add_Vector(
      A=<-1 5 0>
      B=&MergeHelper_VectorSwitch(0=<0 0 -2>, 1=<0 0 0>, 2=?, 3=?, Index=$Particles.PartitionParticles.Partition, Num=::NewEnumerator0)
     )
     Float=&SelectFloatFromArray(Direct Array Index=$Particles.PartitionParticles.Partition, Float Selection Array=DataInterface() {propertyValues {FloatData [1,-1]}})
    )
    Skeletal Mesh=$System.Skeletal Mesh
   ) {module=SkeletalMeshLocation001}
   &AddVelocity(
    Velocity=/Niagara/DynamicInputs/Vectors/RotateVector(
     Vector To Rotate:Vector3f=&Multiply_Vector(
      A=&RandomRangeVector(Minimum=<-1 0 2>, Maximum=<1 -4 6>)
      B=&MergeHelper_VectorSwitch(0=<-1 1 1>, 1=<1 1 1>, 2=?, 3=?, Index=$Particles.PartitionParticles.Partition, Num=::NewEnumerator0)
     )
     Yaw:NiagaraFloat=&MergeHelper_FloatSwitch(0=35, 1=-5, 2=?, 3=?, Index=$Particles.PartitionParticles.Partition, Num=::NewEnumerator0)
     Pitch:NiagaraFloat=0
     Roll:NiagaraFloat=0
     Rotation Rate:NiagaraFloat=1
    )
    Velocity Speed Scale=&RandomRangeFloat(Minimum=20, Maximum=30)
   ) {module=AddVelocity003}
   &InitialMeshOrientation(Use Rotation Vector=false)
   &KillParticles(Kill Particles=&Boolean_OrOperation(
     Bool1=&Boolean_AndOperation(
      Bool1=&SetBoolByFloatComparison(
       A=&MakeFloatFromVector(Vector=$User.ShoeScale_R, Channel=::NewEnumerator0)
       B=0.5
       Comparison Type=::NewEnumerator6
      )
      Bool2=&SetBoolByIntComparison(A=$Particles.PartitionParticles.Partition, B=0, Comparison Type=::NewEnumerator4)
     )
     Bool2=&Boolean_AndOperation(
      Bool1=&SetBoolByFloatComparison(
       A=&MakeFloatFromVector(Vector=$User.ShoeScale_L, Channel=::NewEnumerator0)
       B=0.5
       Comparison Type=::NewEnumerator6
      )
      Bool2=&SetBoolByIntComparison(A=$Particles.PartitionParticles.Partition, B=1, Comparison Type=::NewEnumerator4)
     )
    ))
  ]
  particleUpdateScript [
   &ParticleState()
   &ScaleMeshSize(Scale Factor=&Multiply_VectorByFloat(
     Vector=<1 1 1>
     Float=&FloatFromCurve(FloatCurve=DataInterface() {propertyValues {Curve {keys [
          <at 0 0.3>
          {arriveTangent=-0.375, interpMode=RCIM_Cubic, leaveTangent=-0.375, time=0.2, value=1}
          {arriveTangent=-4.4727, interpMode=RCIM_Cubic, leaveTangent=-4.4727, tangentMode=RCTM_User, time=0.8, value=0}
         ]}}})
    ))
   &SkeletalMeshLocation(
    Bone / Socket Index=$Particles.PartitionParticles.Partition
    Bone Sampling Mode=::NewEnumerator4
    MaskSpawnGroup=false
    Orientation Sampling=::NewEnumerator1
    Position Sampling=::NewEnumerator1
    Sampled Position Offset=&Multiply_Vector(
     A=<4 4 0>
     B=&MergeHelper_VectorSwitch(0=<-1 1 0>, 1=<1 -1 0>, 2=?, 3=?, Index=$Particles.PartitionParticles.Partition, Num=::NewEnumerator0)
    )
    Skeletal Mesh=$System.Skeletal Mesh
   )
   &UpdateMeshOrientation(Rotation Rate=0.3, Rotation Vector=<0 1 0>)
   &Drag(
    Drag=&FloatFromCurve(
     FloatCurve=DataInterface() {propertyValues {Curve {keys [{arriveTangent=3.4423, interpMode=RCIM_Cubic, leaveTangent=3.4423, tangentMode=RCTM_User, time=0, value=0}, <at 1 1 RCIM_Cubic>]}}}
     Scale Curve=10
    )
    Use Linear Drag=true
    Use Rotational Drag=true
   )
   &SolveForcesAndVelocity(Clamp Velocity=false, Limit Acceleration=false)
   &ScaleColor(
    Scale RGB=&MakeVectorFromLinearColorRGB(Color=&LerpLinearColors(StartColor=<rgba 1 1 1 1>, EndColor=$System.TODColor, LerpFactor=1))
    ScaleA=true
    ScaleRGB=true
   )
  ]
  rendererClasses [&NiagaraMeshRendererProperties]
  renderers [{rendererClass=&NiagaraMeshRendererProperties, rendererIndex=0}]
 }
 RainbowTrail {
  emitterUpdateScript [
   &EmitterState(
    Inactive Response=::NewEnumerator1
    Life Cycle Mode=::NewEnumerator1
    Loop Behavior=::NewEnumerator1
    Loop Duration Mode=::NewEnumerator1
    UseLoopDelay=false
   )
   &SpawnBurst_Instantaneous(
    Spawn Count=&Multiply_Int(A=2, B=$System.ParticlePerRibbon)
    Spawn Probability=&MakeCustomFloatFromBool(Bool=$User.HideFootFX, True Float=0, False Float=1)
    Use Spawn Probability=true
    UseLoopCountLimit=false
   )
   /SetVariables(Emitter.VelocityLerp:NiagaraFloat=&FN_FrontEndFloatModifier(
     Gameplay=&FloatFromCurve(
      FloatCurve=DataInterface() {propertyValues {Curve {keys [<at 50 0>,<at 200 1>]}}}
      CurveIndex=&VectorLength(Vector=$Engine.Owner.Velocity)
     )
     Front End=1
    )) {isSetParameters=true}
   /SetVariables(
    Emitter.Scaling:NiagaraFloat=&Lerp_Float(A=1, B=2, Alpha=$Emitter.VelocityLerp)
    Emitter.ShapeSize:NiagaraFloat=&Lerp_Float(
     A=&Lerp_Float(A=3, B=6, Alpha=$Emitter.VelocityLerp)
     B=0
     Alpha=&FloatFromCurve(
      FloatCurve=DataInterface() {propertyValues {Curve {keys [<at 1.2 0 RCIM_Cubic>,<at 1.7 1 RCIM_Cubic>]}}}
      CurveIndex=&Add_Float(
       A=$Emitter.Age
       B=&MakeCustomFloatFromBool(Bool=$User.bisPrimaryPreviewItem, True Float=-1, False Float=0)
      )
     )
    )
   ) {isSetParameters=true}
  ]
  particleSpawnScript [
   &InitializeParticle(
    Lifetime=5
    Ribbon Width=&Lerp_Float(A=$System.TrailWidthMedium, B=$System.TrailWidthLarge, Alpha=$System.ScaleAlpha)
    Ribbon Width Mode=::NewEnumerator1
    UsePositionOffset=false
    Write Material Random=true
   )
   &PartitionParticles(
    Normalized Index In Partition Includes One=true
    Particles Per Partition:NiagaraInt32=$System.ParticlePerRibbon
    Partition Mode:ENiagara_PartitionMode_v2=::NewEnumerator1
    Write Partition Index to Ribbon Link Order:ENiagara_AttributeSamplingApplyOutput=::NewEnumerator0
    Write Partition to Ribbon ID=::NewEnumerator0
   )
   /SetVariables(Particles.RibbonU0Override=$Particles.PartitionParticles.Normalized Index In Partition) {isSetParameters=true}
   /SetVariables() {isSetParameters=true}
   &KillParticles(Kill Particles=&Boolean_OrOperation(
     Bool1=&Boolean_AndOperation(
      Bool1=&SetBoolByFloatComparison(
       A=&MakeFloatFromVector(Vector=$User.ShoeScale_R, Channel=::NewEnumerator0)
       B=0.5
       Comparison Type=::NewEnumerator6
      )
      Bool2=&SetBoolByIntComparison(A=$Particles.PartitionParticles.Partition, B=0, Comparison Type=::NewEnumerator4)
     )
     Bool2=&Boolean_AndOperation(
      Bool1=&SetBoolByFloatComparison(
       A=&MakeFloatFromVector(Vector=$User.ShoeScale_L, Channel=::NewEnumerator0)
       B=0.5
       Comparison Type=::NewEnumerator6
      )
      Bool2=&SetBoolByIntComparison(A=$Particles.PartitionParticles.Partition, B=1, Comparison Type=::NewEnumerator4)
     )
    ))
  ]
  particleUpdateScript [
   &ParticleState(Kill Particles When Lifetime Has Elapsed=false)
   &SkeletalMeshLocation(
    Bone / Socket Index=$Particles.PartitionParticles.Partition
    Bone Sampling Mode=::NewEnumerator4
    MaskSpawnGroup=false
    Orientation Sampling=::NewEnumerator1
    Position Sampling=::NewEnumerator2
    Sampled Position Offset=&Multiply_VectorByFloat(
     Vector=&Add_Vector(
      A=&Lerp_Vector(A=<-1 5 0>, B=<-1 5 0>, Alpha=$System.ScaleAlpha)
      B=&MergeHelper_VectorSwitch(0=<0 0 -2>, 1=<0 0 0>, 2=?, 3=?, Index=$Particles.PartitionParticles.Partition, Num=::NewEnumerator0)
     )
     Float=&SelectFloatFromArray(Direct Array Index=$Particles.PartitionParticles.Partition, Float Selection Array=DataInterface() {propertyValues {FloatData [1,-1]}})
    )
    Skeletal Mesh=$System.Skeletal Mesh
   )
   &InterpolateOverTime(Rate Of Change=9, Target Position Value:NiagaraPosition=$StackContext.SkeletalMeshLocation.SampledPosition, Variable Type=::NewEnumerator6)
   /SetVariables(Particles.AnimationAlpha:NiagaraFloat=&ClampFloat(
     Float=&Add_Float(
      A=&Subtract_Float(A=&Multiply_Float(A=$Particles.PartitionParticles.Normalized Index In Partition, B=0.25), B=0.25)
      B=&Multiply_Float(
       A=$Emitter.Age
       B=&MakeCustomFloatFromBool(Bool=$User.bisPrimaryPreviewItem, True Float=1, False Float=3)
      )
     )
     Max=$Particles.PartitionParticles.Normalized Index In Partition
    )) {isSetParameters=true}
   /SetVariables(Particles.TrailStretchVector:Vector3f=&Multiply_Vector(
     A=&Subtract_Position(A=$StackContext.InterpolateOverTime.InterpolateOverTime_Position.Moving Average, B=$StackContext.SkeletalMeshLocation.SampledPosition)
     B=<1 1 0.666>
    )) {isSetParameters=true}
   /SetVariables(Particles.Position=&AddVectorToPosition(
     Position=$Particles.SkeletalMeshLocation.SampledPosition
     Vector=&Multiply_Vector(
      A=&Add_Vector(
       A=&RotateVectorByQuaternion(
        Vector To Rotate=&Multiply_VectorByFloat(
         Vector=&Add_Vector(
          A=&Lerp_Vector(
           A=<0 0 0>
           B=&Multiply_VectorByFloat(
            Vector=&Multiply_VectorByFloat(Vector=<0 1 0>, Float=1)
            Float=&Lerp_Float(A=$System.TrailLengthMedium, B=$System.TrailLengthLarge, Alpha=$System.ScaleAlpha)
           )
           Alpha=$Particles.AnimationAlpha
          )
          B=&Lerp_Vector(
           A=<-5 0 0>
           B=<0 0 0>
           Alpha=&MakeCustomFloatFromBool(
            Bool=$User.bisPrimaryPreviewItem
            True Float=&FloatFromCurve(
             FloatCurve=DataInterface() {propertyValues {Curve {keys [
                 {arriveTangent=-8.084, interpMode=RCIM_Cubic, leaveTangent=-8.084, tangentMode=RCTM_User, time=0.004, value=1}
                 {arriveTangent=3.4399, interpMode=RCIM_Cubic, leaveTangent=3.4399, tangentMode=RCTM_User, time=1, value=0.5}
                ]}}}
             CurveIndex=$Particles.AnimationAlpha
            )
            False Float=&FloatFromCurve(
             FloatCurve=DataInterface() {propertyValues {Curve {keys [
                 {arriveTangent=-8.084, interpMode=RCIM_Cubic, leaveTangent=-8.084, tangentMode=RCTM_User, time=0.004, value=1}
                 {arriveTangent=3.4399, interpMode=RCIM_Cubic, leaveTangent=3.4399, tangentMode=RCTM_User, time=1, value=1}
                ]}}}
             CurveIndex=$Particles.AnimationAlpha
            )
           )
          )
         )
         Float=&Lerp_Float(A=1, B=-1, Alpha=&MakeFloatFromInt(Int=$Particles.PartitionParticles.Partition))
        )
        Quaternion=$StackContext.SkeletalMeshLocation.SampledRotation
       )
       B=&Multiply_VectorByFloat(
        Vector=&MakeCustomVectorFromBool(
         Bool=$User.bisPrimaryPreviewItem
         True Vector=&RotateVectorByQuaternion(
          Vector To Rotate=&RotateVector(Vector To Rotate:Vector3f=<0 -20 0>, Yaw:NiagaraFloat=30, Pitch:NiagaraFloat=0, Roll:NiagaraFloat=0, Rotation Rate:NiagaraFloat=1)
          Quaternion=$Engine.Owner.Rotation
         )
         False Vector=$StackContext.TrailStretchVector
        )
        Float=&FloatFromCurve(
         FloatCurve=DataInterface() {propertyValues {Curve {keys [<at 0 0>, {leaveTangent=2.6963, time=1, value=1}]}}}
         CurveIndex=$Particles.Initial.PartitionParticles.Normalized Index In Partition
        )
       )
      )
      B=&MergeHelper_VectorSwitch(0=$User.ShoeScale_R, 1=$User.ShoeScale_L, 2=?, 3=?, Index=$Particles.PartitionParticles.Partition, Num=::NewEnumerator0)
     )
    )) {isSetParameters=true}
   &ScaleRibbonWidth(Ribbon Width Scale=&Lerp_Float(
     A=1
     B=2
     Alpha=&FloatFromCurve(
      FloatCurve=DataInterface() {propertyValues {Curve {keys [
          {interpMode=RCIM_Cubic, tangentMode=RCTM_User, time=0, value=0}
          {arriveTangent=2.2695, interpMode=RCIM_Cubic, leaveTangent=2.2695, tangentMode=RCTM_User, time=1, value=1}
         ]}}}
      CurveIndex=$Particles.PartitionParticles.Normalized Index In Partition
      Scale Curve=&MakeFloatFromVector(
       Vector=&MergeHelper_VectorSwitch(0=$User.ShoeScale_R, 1=$User.ShoeScale_L, 2=?, 3=?, Index=$Particles.PartitionParticles.Partition, Num=::NewEnumerator0)
       Channel=::NewEnumerator0
      )
     )
    ))
   &ScaleColor(
    Scale Alpha=&FloatFromCurve(FloatCurve=DataInterface() {propertyValues {Curve {keys [<at 1.5 1>,<at 1.7 1>]}}}, CurveIndex=$Emitter.Age)
    Scale RGB=&MakeVectorFromLinearColorRGB(Color=$System.TODColor)
    ScaleA=true
    ScaleRGB=true
   )
   /SetVariables(Particles.TextureOffset:NiagaraFloat=&Add_Float(
     A=$Particles.TextureOffset
     B=&Multiply_Float(A=$Engine.DeltaTime, B=&Lerp_Float(A=-0.75, B=-2, Alpha=$Emitter.VelocityLerp))
    )) {isSetParameters=true}
   &DynamicMaterialParameters(
    Index 0 Param 1=&Add_Float(
     A=&RandomRangeFloat(
      Minimum=0
      Maximum=100
      Randomness Mode=::NewEnumerator1
      Random Seed=$Particles.Initial.PartitionParticles.Partition
      Fixed Random Seed=true
     )
     B=$Particles.TextureOffset
    )
    Module.Index0Float0Write=true
    Module.Index0Float1Write=true
    Module.Index0Float2Write=true
    Module.Index0Float3Write=true
   )
  ]
  rendererClasses [&NiagaraRibbonRendererProperties]
  renderers [{rendererClass=&NiagaraRibbonRendererProperties}]
 }
 Smoke_Circular {
  emitterUpdateScript [
   &EmitterState(Life Cycle Mode=::NewEnumerator1, Loop Behavior=::NewEnumerator1, Loop Duration=1, UseLoopDelay=false)
   &SpawnBurst_Instantaneous(
    Spawn Count=2
    Spawn Probability=&MakeCustomFloatFromBool(Bool=&Boolean_OrOperation(Bool1=$User.HideFootFX, Bool2=$User.bisPrimaryPreviewItem), True Float=0, False Float=1)
    Use Spawn Probability=true
    UseLoopCountLimit=false
   )
  ]
  particleSpawnScript [
   &InitializeParticle(
    Color=<rgba 1.5 1.5 1.5 1>
    Lifetime Max=0.5
    Lifetime Min=0.3
    Lifetime Mode=::NewEnumerator1
    Mesh Scale Mode=::NewEnumerator1
    Mesh Uniform Scale Max=3
    Mesh Uniform Scale Min=2
    UsePositionOffset=false
    Write Material Random=true
   )
   &InitialMeshOrientation(
    Mesh Axis To Orient=&RandomRangeVector(Minimum=<-1 -1 0>, Maximum=<1 1 0>)
    Mesh Orientation Mode=::NewEnumerator2
    Use Rotation Vector=false
   )
   &AddVelocity(Velocity=<0 0 20>)
  ]
  particleUpdateScript [
   &ParticleState()
   &ScaleMeshSize(Scale Factor=&VectorFromFloat(Value=&FloatFromCurve(FloatCurve=DataInterface() {propertyValues {Curve {
         defaultValue=1
         keys [
          {arriveTangent=2.0915, interpMode=RCIM_Cubic, leaveTangent=2.0915, tangentMode=RCTM_User, time=0, value=0}
          {arriveTangent=0.0226, interpMode=RCIM_Cubic, leaveTangent=0.0226, tangentMode=RCTM_User, time=1, value=1}
         ]
        }}})))
   &SolveForcesAndVelocity(Clamp Velocity=false, Limit Acceleration=false)
   &ScaleColor(
    Scale Alpha=&FloatFromCurve(FloatCurve=DataInterface() {propertyValues {Curve {keys [
         {arriveTangent=-1.018, interpMode=RCIM_Cubic, leaveTangent=-1.018, tangentMode=RCTM_User, time=0, value=1}
         {arriveTangent=-0.4774, interpMode=RCIM_Cubic, leaveTangent=-0.4774, tangentMode=RCTM_User, time=1, value=0.3}
        ]}}})
    Scale RGB=&MakeVectorFromLinearColorRGB(Color=&LerpLinearColors(StartColor=<rgba 1 1 1 1>, EndColor=$System.TODColor, LerpFactor=1))
    ScaleA=true
    ScaleRGB=true
   ) {module=ScaleColor001}
  ]
  rendererClasses [&NiagaraMeshRendererProperties]
  renderers [{rendererClass=&NiagaraMeshRendererProperties, rendererIndex=0}]
 }
 Smoke_Puffs {
  emitterUpdateScript [
   &EmitterState(Life Cycle Mode=::NewEnumerator1, Loop Behavior=::NewEnumerator1, Loop Duration=1, UseLoopDelay=false)
   &SpawnBurst_Instantaneous(
    Spawn Count=&RandomRangeInt(Minimum=15, Maximum=25)
    Spawn Probability=&MakeCustomFloatFromBool(Bool=&Boolean_OrOperation(Bool1=$User.HideFootFX, Bool2=$User.bisPrimaryPreviewItem), True Float=0, False Float=1)
    Use Spawn Probability=true
    UseLoopCountLimit=false
   )
   /SetVariables(Emitter.OutlineWidth:NiagaraFloat=&Lerp_Float(
     A=0.7
     B=1.5
     Alpha=&FloatFromCurve(
      FloatCurve=DataInterface() {propertyValues {Curve {keys [<at 0 0 RCIM_Cubic>, {arriveTangent=3.5879, interpMode=RCIM_Cubic, leaveTangent=3.5879, tangentMode=RCTM_User, time=0.6, value=1}]}}}
      CurveIndex=$Emitter.Age
     )
    )) {isSetParameters=true}
  ]
  particleSpawnScript [
   &InitializeParticle(
    Color=<rgba 1.5 1.5 1.5 1>
    Lifetime Max=0.6
    Lifetime Min=0.3
    Lifetime Mode=::NewEnumerator1
    Mesh Scale Mode=::NewEnumerator1
    Mesh Uniform Scale Max=1
    Mesh Uniform Scale Min=0.5
    UsePositionOffset=false
    Write Material Random=true
   )
   &ShapeLocation(
    Handle Radius=0.3
    Large Radius=1
    MaskSpawnGroup=false
    Override Seed=false
    Rotation Axis=<0 0 1>
    Rotation Mode=::NewEnumerator1
    Shape Primitive=::NewEnumerator3
    UseOwnerScale=true
   )
   &AddVelocity(
    Module.Use Falloff Distance=false
    Override Seed=false
    Velocity Mode=::NewEnumerator1
    Velocity Speed=&RandomRangeFloat(Minimum=250, Maximum=400)
   )
   &InitialMeshOrientation(Use Rotation Vector=false)
  ]
  particleUpdateScript [
   &ParticleState()
   &ScaleMeshSize(Scale Factor=&Multiply_VectorByFloat(
     Vector=<1 1 1>
     Float=&FloatFromCurve(FloatCurve=DataInterface() {propertyValues {Curve {keys [
          <at 0 0.5>
          {arriveTangent=-0.625, interpMode=RCIM_Cubic, leaveTangent=-0.625, time=0.2, value=1}
          {arriveTangent=-4.4727, interpMode=RCIM_Cubic, leaveTangent=-4.4727, tangentMode=RCTM_User, time=0.8, value=0}
         ]}}})
    ))
   &UpdateMeshOrientation(Rotation Rate=0.3, Rotation Vector=<0 1 0>)
   &Drag(
    Drag=&FloatFromCurve(
     FloatCurve=DataInterface() {propertyValues {Curve {keys [{arriveTangent=3.4423, interpMode=RCIM_Cubic, leaveTangent=3.4423, tangentMode=RCTM_User, time=0, value=0}, <at 1 1 RCIM_Cubic>]}}}
     Scale Curve=10
    )
    Use Linear Drag=true
    Use Rotational Drag=true
   )
   &SolveForcesAndVelocity(Clamp Velocity=false, Limit Acceleration=false)
   &ScaleColor(
    Scale RGB=&MakeVectorFromLinearColorRGB(Color=&LerpLinearColors(StartColor=<rgba 1 1 1 1>, EndColor=$System.TODColor, LerpFactor=1))
    ScaleA=true
    ScaleRGB=true
   )
  ]
  rendererClasses [&NiagaraMeshRendererProperties]
  renderers [{rendererClass=&NiagaraMeshRendererProperties, rendererIndex=0}]
 }
 Sparkles {
  emitterUpdateScript [
   &EmitterState(Life Cycle Mode=::NewEnumerator1, Loop Behavior=::NewEnumerator1, Loop Duration=1, UseLoopDelay=false)
   &SpawnBurst_Instantaneous(
    Spawn Count=&RandomRangeInt(Minimum=5, Maximum=10)
    Spawn Probability=&MakeCustomFloatFromBool(Bool=$User.HideFootFX, True Float=0, False Float=1)
    Use Spawn Probability=true
    UseLoopCountLimit=false
   )
  ]
  particleSpawnScript [
   &InitializeParticle(
    Color=&RandomRangeLinearColor(Minimum=<rgba 1.5 1.1341 0.3203 1>, Maximum=<rgba 1.5 0.902 0.1271 1>)
    Lifetime=&MakeCustomFloatFromBool(
     Bool=$User.bisPrimaryPreviewItem
     True Float=&RandomRangeFloat(Minimum=0.5, Maximum=1)
     False Float=&RandomRangeFloat(Minimum=0.4, Maximum=0.8)
    )
    Sprite Size Mode=::NewEnumerator3
    Uniform Sprite Size=&RandomRangeFloat(Minimum=10, Maximum=20)
    UsePositionOffset=false
    Write Material Random=true
   )
   &ShapeLocation(
    Handle Radius=0.3
    Large Radius=10
    MaskSpawnGroup=false
    Override Seed=false
    Rotation Axis=<0 0 1>
    Rotation Mode=::NewEnumerator1
    Shape Primitive=::NewEnumerator3
    UseOwnerScale=true
   )
   &AddVelocity(
    Module.Use Falloff Distance=false
    Origin Offset=<0 0 -5>
    Override Seed=false
    Velocity Mode=::NewEnumerator1
    Velocity Speed=&RandomRangeFloat(Minimum=100, Maximum=200)
   )
   &AddVelocity(Velocity Speed Scale=&RandomRangeFloat(Minimum=1, Maximum=2)) {module=AddVelocity001}
   &CameraOffset(Camera Offset Amount=10)
  ]
  particleUpdateScript [
   &ParticleState()
   /SetVariables(Particles.Wave:NiagaraFloat=&Waveform(
     Waveform Count=::NewEnumerator1
     Global Input Position=&RandomRangeFloat(Minimum=0, Maximum=10)
     "Waveform [1]"=::NewEnumerator0
     "[1] Frequency"=4
     "[1] Amplitude Min/Max"=&RandomRangeVector2D(Minimum=<0.3 1>, Maximum=<0.5 1>)
     "Waveform [2] Blend"=::NewEnumerator2
     "Waveform [2]"=::NewEnumerator8
     "[2] Input Position"=0
     "[2] Phase"=$Emitter.Age
     "[2] Phase Scale"=1
     "[2] Compound Frequency"=2
     "[2] Frequency"=8
     "[2] Amplitude Min/Max"=<0.8 1>
     "[2] Scale"=1
     "[2] Bias"=0
     "[2] Exponent"=1
     "[2] Pulse Duration"=0.1
     "[2] Clip Triangle Min/Max"=<0 0>
     "[2] Random Seed"=0
    )) {isSetParameters=true}
   &ScaleSpriteSize(
    Scale Sprite Size Mode=::NewEnumerator0
    Uniform Scale Factor=&Multiply_Float(
     A=$Particles.Wave
     B=&FloatFromCurve(FloatCurve=DataInterface() {propertyValues {Curve {keys [
          {arriveTangent=14.864, interpMode=RCIM_Cubic, leaveTangent=14.864, tangentMode=RCTM_User, time=0, value=0}
          {arriveTangent=2, leaveTangent=2, time=0.1, value=1}
          {arriveTangent=-1.1111, interpMode=RCIM_Cubic, leaveTangent=-1.1111, time=0.5, value=1}
          <at 1 0 RCIM_Cubic>
         ]}}})
    )
   )
   &Drag(
    Drag=&FloatFromCurve(
     FloatCurve=DataInterface() {propertyValues {Curve {keys [{arriveTangent=3.4423, interpMode=RCIM_Cubic, leaveTangent=3.4423, tangentMode=RCTM_User, time=0, value=0}, <at 1 1 RCIM_Cubic>]}}}
     Scale Curve=10
    )
    Use Linear Drag=true
    Use Rotational Drag=true
   )
   &SolveForcesAndVelocity(Clamp Velocity=false, Limit Acceleration=false)
   &DynamicMaterialParameters(
    Index 0 Param 1=&Add_Float(
     A=&RandomRangeFloat(Minimum=0, Maximum=1)
     B=&FloatFromCurve(
      FloatCurve=DataInterface() {propertyValues {Curve {keys [
          <at 0 0 RCIM_Cubic>
          {arriveTangent=1.5947, interpMode=RCIM_Cubic, leaveTangent=1.5947, tangentMode=RCTM_User, time=0.2557, value=0.5257}
          <at 1 1 RCIM_Cubic>
         ]}}}
      CurveIndex=$Particles.Age
      Scale Curve=0.25
     )
    )
    Module.Param0WriteEnabled=true
    Module.Param1WriteEnabled=true
    Module.Param2WriteEnabled=true
    Module.Param3WriteEnabled=true
   ) {module=DynamicMaterialParameters001}
   &ScaleColor(
    Scale RGB=&VectorFromFloat(Value=&Multiply_Float(A=$Particles.Wave, B=&RandomRangeFloat(Minimum=1, Maximum=2)))
    ScaleA=true
    ScaleRGB=true
   )
   &FN_Sprite_Attributes(
    Color=&MakeLinearColorFromVectorAndFloat(
     "Vector (RGB)"=&MakeVectorFromLinearColorRGB(Color=<rgba 1 0.431 0 1>)
     "Float (Alpha)"=&FloatFromCurve(
      FloatCurve=DataInterface() {propertyValues {Curve {
         defaultValue=0
         keys [
          {interpMode=RCIM_Cubic, tangentMode=RCTM_User, time=0, value=1}
          {arriveTangent=-2.2441, interpMode=RCIM_Cubic, leaveTangent=-2.2441, tangentMode=RCTM_User, time=1, value=0}
         ]
        }}}
      Scale Curve=&MakeCustomFloatFromBool(
       Bool=$User.bisPrimaryPreviewItem
       True Float=&RandomRangeFloat(Minimum=0.7, Maximum=1)
       False Float=&RandomRangeFloat(Minimum=1.5, Maximum=2)
      )
     )
    )
    Create Color Attribute=true
    Create SpriteSize Attribute=true
    SpriteSize=&Multiply_Vector2D_ByFloat(Float=1.5, Vector2D=$Particles.SpriteSize)
    Write to Module Namespace=true
   )
  ]
  rendererClasses [&NiagaraSpriteRendererProperties]
  renderers [
   {rendererClass=&NiagaraSpriteRendererProperties}
   {rendererClass=&NiagaraSpriteRendererProperties, rendererIndex=1}
  ]
 }
}
summary {
 systemName=NS_Shoe_SoleVapor_Landing
 emitters [
  {emitterName=Smoke_Puffs, rendererClasses [&NiagaraMeshRendererProperties]}
  {emitterName=Smoke_Circular, rendererClasses [&NiagaraMeshRendererProperties]}
  {emitterName=RainbowTrail, rendererClasses [&NiagaraRibbonRendererProperties]}
  {emitterName=ColoredSmokePuffs, rendererClasses [&NiagaraMeshRendererProperties]}
  {emitterName=Sparkles, rendererClasses [&NiagaraSpriteRendererProperties]}
 ]
 userVariables [
  User.bIsFrontEnd {type=NiagaraBool}
  User.bIsFrontEndPreview {type=NiagaraBool}
  User.bisPrimaryPreviewItem {type=NiagaraBool}
  User.HideFootFX {type=NiagaraBool}
  User.ShoeScale_L {type=Vector3f, description="", defaultValue {struct=Vector3f, value=<1 1 1>}}
  User.ShoeScale_R {type=Vector3f, description="", defaultValue {struct=Vector3f, value=<1 1 1>}}
 ]
}
